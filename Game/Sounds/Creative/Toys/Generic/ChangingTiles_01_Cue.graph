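# Sound cue for a changing-tiles effect on a Creative toy, built from two reused Grumpy panel-change waves.
# A Mixer plays one wave at slightly lowered random pitch, layered with the other, which is briefly delayed, pitched up and mixed at half volume.
# The quiet cue uses the Creative gadgets sound class, overriding attenuation with listener focus and concurrency with a small voice cap.
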SoundCue /Game/Sounds/Creative/Toys/Generic/ChangingTiles_01_Cue
root=Mixer_0
nodes [
 {id=Delay_0, children [Modulator_2], properties {delayMax=0.2, delayMin=0.2}}
 {id=Mixer_0, children [Modulator_1,Delay_0], properties {inputVolume [1,0.5]}}
 {id=Modulator_1, children [WavePlayer_4], properties {pitchMax=1, pitchMin=0.9}}
 {id=Modulator_2, children [WavePlayer_0], properties {pitchMax=1.3, pitchMin=1.2}}
 {id=WavePlayer_0, properties {soundWaveAssetPtr=/Game/Athena/Sounds/Events/Dopey/Grumpy/Grumpy_Panel_Change_03}}
 {id=WavePlayer_4, properties {soundWaveAssetPtr=/Game/Athena/Sounds/Events/Dopey/Grumpy/Grumpy_Panel_Change_02}}
]
properties {
 attenuationSettings=/Game/Sounds/Creative/Attenuations/Fort_Creative_SmallToy_Att
 bOverrideAttenuation=true
 bOverrideConcurrency=true
 maxDistance=1515
 soundClassObject=&Fort_Gadgets_Creative
 volumeMultiplier=0.25
 attenuationOverrides {
  bEnableListenerFocus=true
  falloffDistance=1000
  focusAzimuth=20
  nonFocusAzimuth=50
  nonFocusVolumeAttenuation=0.4
  attenuationShapeExtents {x=515}
 }
 concurrencyOverrides {bVolumeScaleCanRelease=true, maxCount=5, volumeScaleReleaseTime=0.25}
}
