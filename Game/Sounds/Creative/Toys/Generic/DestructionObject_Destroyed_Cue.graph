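# Sound cue for the explosion of a destroyed Creative toy destruction object, varied by object type.
# A switch on the ObjectIndex parameter selects among ten inputs, each a random pick from C4, Flintlock launcher, ballistic turret or Corrupted Aura explosion waves.
# The Flintlock branches pass through modulators that lower their volume.
# Attenuation and concurrency overrides add distance low-pass filtering and cap simultaneous plays, stopping the farthest first.
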SoundCue /Game/Sounds/Creative/Toys/Generic/DestructionObject_Destroyed_Cue
root=Switch_0
nodes [
 {id=Modulator_0, children [Random_16], properties {volumeMax=0.8, volumeMin=0.8}}
 {id=Modulator_1, children [Random_17], properties {volumeMax=0.8, volumeMin=0.8}}
 {id=Modulator_2, children [Random_18], properties {volumeMax=0.8, volumeMin=0.8}}
 {id=Random_10, children [WavePlayer_23, WavePlayer_24, WavePlayer_25], properties {weights [1,1,1]}}
 {id=Random_11, children [WavePlayer_26, WavePlayer_27, WavePlayer_28], properties {weights [1,1,1]}}
 {id=Random_12, children [WavePlayer_29, WavePlayer_30, WavePlayer_31], properties {weights [1,1,1]}}
 {id=Random_16, children [WavePlayer_38, WavePlayer_39], properties {weights [1,1]}}
 {id=Random_17, children [WavePlayer_40, WavePlayer_41], properties {weights [1,1]}}
 {id=Random_18, children [WavePlayer_42, WavePlayer_43], properties {weights [1,1]}}
 {id=Random_5, children [WavePlayer_11, WavePlayer_12], properties {weights [1,1]}}
 {id=Random_6, children [WavePlayer_13, WavePlayer_14], properties {weights [1,1]}}
 {id=Random_9, children [WavePlayer_20, WavePlayer_21, WavePlayer_22], properties {weights [1,1,1]}}
 {
  id=Switch_0
  children [Random_11,Modulator_0,Modulator_1,Modulator_2,Random_11,Random_12,Random_5,Random_9,Random_10,Random_6]
  properties {intParameterName=ObjectIndex}
 }
 {id=WavePlayer_11, properties {soundWaveAssetPtr=/Game/Sounds/Fort_Traps/MountedTurret/Turret_Ballistic_Explode_02}}
 {id=WavePlayer_12, properties {soundWaveAssetPtr=/Game/Sounds/Fort_Traps/MountedTurret/Turret_Ballistic_Explode_01}}
 {id=WavePlayer_13, properties {soundWaveAssetPtr=&Turret_Ballistic_Explode_02}}
 {id=WavePlayer_14, properties {soundWaveAssetPtr=&Turret_Ballistic_Explode_01}}
 {id=WavePlayer_20, properties {soundWaveAssetPtr=/Game/Sounds/Fort_Gadgets/Ninja/Perk_Corrupted_Aura/Perk_Corrupted_Aura_Explosion_01}}
 {id=WavePlayer_21, properties {soundWaveAssetPtr=/Game/Sounds/Fort_Gadgets/Ninja/Perk_Corrupted_Aura/Perk_Corrupted_Aura_Explosion_03}}
 {id=WavePlayer_22, properties {soundWaveAssetPtr=/Game/Sounds/Fort_Gadgets/Ninja/Perk_Corrupted_Aura/Perk_Corrupted_Aura_Explosion_04}}
 {id=WavePlayer_23, properties {soundWaveAssetPtr=&Perk_Corrupted_Aura_Explosion_01}}
 {id=WavePlayer_24, properties {soundWaveAssetPtr=&Perk_Corrupted_Aura_Explosion_03}}
 {id=WavePlayer_25, properties {soundWaveAssetPtr=&Perk_Corrupted_Aura_Explosion_04}}
 {id=WavePlayer_26, properties {soundWaveAssetPtr=/Game/Athena/Sounds/Items/C4/C4_Explode_New_01}}
 {id=WavePlayer_27, properties {soundWaveAssetPtr=/Game/Athena/Sounds/Items/C4/C4_Explode_New_02}}
 {id=WavePlayer_28, properties {soundWaveAssetPtr=/Game/Athena/Sounds/Items/C4/C4_Explode_New_03}}
 {id=WavePlayer_29, properties {soundWaveAssetPtr=&C4_Explode_New_01}}
 {id=WavePlayer_30, properties {soundWaveAssetPtr=&C4_Explode_New_02}}
 {id=WavePlayer_31, properties {soundWaveAssetPtr=&C4_Explode_New_03}}
 {id=WavePlayer_38, properties {soundWaveAssetPtr=/Game/Sounds/Fort_Weapons/FlintLock/Flintlock_Launcher_BallExplode_01}}
 {id=WavePlayer_39, properties {soundWaveAssetPtr=/Game/Sounds/Fort_Weapons/FlintLock/Flintlock_Launcher_BallExplode_02}}
 {id=WavePlayer_40, properties {soundWaveAssetPtr=&Flintlock_Launcher_BallExplode_01}}
 {id=WavePlayer_41, properties {soundWaveAssetPtr=&Flintlock_Launcher_BallExplode_02}}
 {id=WavePlayer_42, properties {soundWaveAssetPtr=&Flintlock_Launcher_BallExplode_01}}
 {id=WavePlayer_43, properties {soundWaveAssetPtr=&Flintlock_Launcher_BallExplode_02}}
]
properties {
 bOverrideAttenuation=true
 bOverrideConcurrency=true
 maxDistance=6800
 soundClassObject=/Game/Sounds/FortSoundClassesAndModes/Fort_Explosives_Medium_PSM
 volumeMultiplier=0.6
 attenuationOverrides {bAttenuateWithLPF=true, falloffDistance=6000, lPFFrequencyAtMax=100, lPFRadiusMin=2000, stereoSpread=900, attenuationShapeExtents {x=800}}
 concurrencyOverrides {maxCount=3, resolutionRule=StopFarthestThenPreventNew}
}
