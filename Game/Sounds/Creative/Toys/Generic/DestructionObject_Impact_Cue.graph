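# Sound cue for impact sounds on Creative toy destruction objects, chosen by object type.
# A Switch on the ObjectIndex int parameter picks among ten slots, each a Random node over three waves.
# The waves are reused gnome and robot pickaxe hits, oil drum, small metal and light metal impacts, most trimmed by Modulators for volume, oil drums also pitched up.
# It uses bullet-impact attenuation in the Fort_Impacts sound class.
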SoundCue /Game/Sounds/Creative/Toys/Generic/DestructionObject_Impact_Cue
root=Switch_0
nodes [
 {id=Modulator_0, children [Random_10], properties {volumeMax=1.3, volumeMin=1.3}}
 {id=Modulator_1, children [Random_6], properties {volumeMax=0.85, volumeMin=0.85}}
 {id=Modulator_2, children [Random_7], properties {volumeMax=0.85, volumeMin=0.85}}
 {id=Modulator_3, children [Random_4], properties {volumeMax=1.1, volumeMin=1.1}}
 {id=Modulator_4, children [Random_3], properties {volumeMax=1.1, volumeMin=1.1}}
 {id=Modulator_6, children [Random_8], properties {pitchMax=1.1, pitchMin=1.1, volumeMax=0.78, volumeMin=0.78}}
 {id=Modulator_7, children [Random_9], properties {pitchMax=1.1, pitchMin=1.1, volumeMax=0.78, volumeMin=0.78}}
 {id=Random_10, children [WavePlayer_33, WavePlayer_34, WavePlayer_35], properties {weights [1,1,1]}}
 {id=Random_11, children [WavePlayer_36, WavePlayer_37, WavePlayer_38], properties {weights [1,1,1]}}
 {id=Random_12, children [WavePlayer_39, WavePlayer_40, WavePlayer_41], properties {weights [1,1,1]}}
 {id=Random_3, children [WavePlayer_9, WavePlayer_10, WavePlayer_11], properties {weights [1,1,1]}}
 {id=Random_4, children [WavePlayer_12, WavePlayer_13, WavePlayer_14], properties {weights [1,1,1]}}
 {id=Random_6, children [WavePlayer_21, WavePlayer_22, WavePlayer_23], properties {weights [1,1,1]}}
 {id=Random_7, children [WavePlayer_24, WavePlayer_25, WavePlayer_26], properties {weights [1,1,1]}}
 {id=Random_8, children [WavePlayer_27, WavePlayer_28, WavePlayer_29], properties {weights [1,1,1]}}
 {id=Random_9, children [WavePlayer_30, WavePlayer_31, WavePlayer_32], properties {weights [1,1,1]}}
 {
  id=Switch_0
  children [Random_12,Random_12,Modulator_6,Modulator_7,Modulator_4,Modulator_3,Random_11,Modulator_1,Modulator_2,Modulator_0]
  properties {intParameterName=ObjectIndex}
 }
 {id=WavePlayer_10, properties {soundWaveAssetPtr=/Game/Athena/Sounds/Weapons/PickAxes/Gnome/Pickaxes_Gnome_Impact_05}}
 {id=WavePlayer_11, properties {soundWaveAssetPtr=/Game/Athena/Sounds/Weapons/PickAxes/Gnome/Pickaxes_Gnome_Impact_04}}
 {id=WavePlayer_12, properties {soundWaveAssetPtr=/Game/Athena/Sounds/Weapons/PickAxes/Gnome/Pickaxes_Gnome_Impact_06}}
 {id=WavePlayer_13, properties {soundWaveAssetPtr=&Pickaxes_Gnome_Impact_05}}
 {id=WavePlayer_14, properties {soundWaveAssetPtr=&Pickaxes_Gnome_Impact_04}}
 {id=WavePlayer_21, properties {soundWaveAssetPtr=/Game/Athena/Sounds/Weapons/PickAxes/Robot_Trouble/Pickaxe_Robot_Trouble_Impact_02}}
 {id=WavePlayer_22, properties {soundWaveAssetPtr=/Game/Athena/Sounds/Weapons/PickAxes/Robot_Trouble/Pickaxe_Robot_Trouble_Impact_04}}
 {id=WavePlayer_23, properties {soundWaveAssetPtr=/Game/Athena/Sounds/Weapons/PickAxes/Robot_Trouble/Pickaxe_Robot_Trouble_Impact_05}}
 {id=WavePlayer_24, properties {soundWaveAssetPtr=&Pickaxe_Robot_Trouble_Impact_02}}
 {id=WavePlayer_25, properties {soundWaveAssetPtr=&Pickaxe_Robot_Trouble_Impact_04}}
 {id=WavePlayer_26, properties {soundWaveAssetPtr=&Pickaxe_Robot_Trouble_Impact_05}}
 {id=WavePlayer_27, properties {soundWaveAssetPtr=/Game/Sounds/Fort_Impact_Sounds/Demolition/OilDrum/oil_drum_impact_01}}
 {id=WavePlayer_28, properties {soundWaveAssetPtr=/Game/Sounds/Fort_Impact_Sounds/Demolition/OilDrum/oil_drum_impact_02}}
 {id=WavePlayer_29, properties {soundWaveAssetPtr=/Game/Sounds/Fort_Impact_Sounds/Demolition/OilDrum/oil_drum_impact_03}}
 {id=WavePlayer_30, properties {soundWaveAssetPtr=&oil_drum_impact_01}}
 {id=WavePlayer_31, properties {soundWaveAssetPtr=&oil_drum_impact_02}}
 {id=WavePlayer_32, properties {soundWaveAssetPtr=&oil_drum_impact_03}}
 {id=WavePlayer_33, properties {soundWaveAssetPtr=/Game/Sounds/Fort_Impact_Sounds/Demolition/Metal/Metal_Small_Impact06}}
 {id=WavePlayer_34, properties {soundWaveAssetPtr=/Game/Sounds/Fort_Impact_Sounds/Demolition/Metal/Metal_Small_Impact04}}
 {id=WavePlayer_35, properties {soundWaveAssetPtr=/Game/Sounds/Fort_Impact_Sounds/Demolition/Metal/Metal_Small_Impact02}}
 {id=WavePlayer_36, properties {soundWaveAssetPtr=/Game/Sounds/Fort_Weapons/Impacts/MTL_IMPACT_LIGHT_08}}
 {id=WavePlayer_37, properties {soundWaveAssetPtr=/Game/Sounds/Fort_Weapons/Impacts/MTL_IMPACT_LIGHT_07}}
 {id=WavePlayer_38, properties {soundWaveAssetPtr=/Game/Sounds/Fort_Weapons/Impacts/MTL_IMPACT_LIGHT_03}}
 {id=WavePlayer_39, properties {soundWaveAssetPtr=&MTL_IMPACT_LIGHT_08}}
 {id=WavePlayer_40, properties {soundWaveAssetPtr=&MTL_IMPACT_LIGHT_07}}
 {id=WavePlayer_41, properties {soundWaveAssetPtr=&MTL_IMPACT_LIGHT_03}}
 {id=WavePlayer_9, properties {soundWaveAssetPtr=&Pickaxes_Gnome_Impact_06}}
]
properties {
 attenuationSettings=/Game/Sounds/Attenuation/Fort_Bullet_Impact_Athena_Att
 maxDistance=4500
 soundClassObject=&Fort_Impacts
 volumeMultiplier=0.6
 attenuationOverrides {
  bEnableFocusInterpolation=true
  bEnableListenerFocus=true
  falloffDistance=300
  focusAttackInterpSpeed=20
  focusReleaseInterpSpeed=10
  focusVolumeAttenuation=1.1
  nonFocusVolumeAttenuation=0.6
  attenuationShapeExtents {x=0}
 }
 concurrencyOverrides {maxCount=3, volumeScale=0.6}
}
