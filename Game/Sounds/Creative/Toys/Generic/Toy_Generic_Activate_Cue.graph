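# Sound cue for the generic creative toy activation sound.
# A single wave player plays one activation wave at reduced volume, routed to the gadgets sound class with small-toy attenuation.
# Overridden concurrency allows one instance per owner, stopping the oldest when retriggered.
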SoundCue /Game/Sounds/Creative/Toys/Generic/Toy_Generic_Activate_Cue
root=WavePlayer_0
nodes [{id=WavePlayer_0, properties {soundWaveAssetPtr=&/Toy_Generic_Activate_01}}]
properties {
 attenuationSettings=/Game/Sounds/Creative/Attenuations/Fort_Creative_SmallToy_Att
 bOverrideConcurrency=true
 maxDistance=3000
 soundClassObject=&Fort_Gadgets
 volumeMultiplier=0.45
 concurrencyOverrides {bLimitToOwner=true, maxCount=1, resolutionRule=StopOldest}
}
